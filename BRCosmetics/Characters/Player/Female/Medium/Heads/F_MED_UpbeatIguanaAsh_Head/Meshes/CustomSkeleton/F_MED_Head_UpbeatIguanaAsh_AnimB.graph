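# Animation Blueprint for the UpbeatIguanaAsh female head cosmetic, adding secondary motion to ears, front hair, jaw and pupils.
# A Control Rig, fed a hang-gliding alpha set by parachute events, runs into chained spring-based AnimDynamics nodes, and the result is cached.
# Two pupil simulations, world-space and root-relative, branch from that cache into a bool blend for output.
# Each update refreshes a gravity override and a front-end-menu flag that excludes emotes.
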AnimBlueprint /BRCosmetics/Characters/Player/Female/Medium/Heads/F_MED_UpbeatIguanaAsh_Head/Meshes/CustomSkeleton/F_MED_Head_UpbeatIguanaAsh_AnimB
target_skeleton=&/Fortnite_Base_Head_UpbeatIguanaAsh_Skel
graphs [
 AnimGraph {nodes [
   n1 = Anim Dynamics - Bone: dyn_pupil_00_l!(ComponentPose=Local To Component(LocalPose=Use cached pose 'CRCache_Head'()), Alpha=1, GravityOverride:Vector="0,0,-980") -> (Pose) {
    angularDampingOverride=1
    angularSpringConstant=20000
    bAngularSpring=true
    bOverrideAngularDamping=true
    bUseGravityOverride=true
    gravityScale=0.35
    simulationSpace=World
    alphaBoolBlend {blendOption=Linear}
    boundBone {boneName=dyn_pupil_00_l}
    componentLinearAccScale=<1 1 1>
    componentLinearVelScale=<1 1 1>
    gravityOverride {z=-980}
    physicsBodyDefinitions [{
      boundBone {boneName=dyn_pupil_00_l}
      boxExtents=<20 30 30>
      constraintSetup {angularLimitsMax=<15 35 35>, angularLimitsMin=<-15 -35 -35>, angularTarget=<1 0 0>}
      localJointOffset=<1 0 0>
     }]
   }
   n2 = Anim Dynamics - Bone: dyn_pupil_00_r!(ComponentPose=@n1, Alpha=1, GravityOverride:Vector="0,0,-980") -> (Pose) {
    angularDampingOverride=1
    angularSpringConstant=20000
    bAngularSpring=true
    bOverrideAngularDamping=true
    bUseGravityOverride=true
    gravityScale=0.35
    simulationSpace=World
    alphaBoolBlend {blendOption=Linear}
    boundBone {boneName=dyn_pupil_00_r}
    componentLinearAccScale=<1 1 1>
    componentLinearVelScale=<1 1 1>
    gravityOverride {z=-980}
    physicsBodyDefinitions [{
      boundBone {boneName=dyn_pupil_00_r}
      boxExtents=<20 30 30>
      constraintSetup {angularLimitsMax=<15 35 35>, angularLimitsMin=<-15 -35 -35>, angularTarget=<1 0 0>}
      localJointOffset=<1 0 0>
     }]
   }
   n3 = Anim Dynamics - Bone: dyn_pupil_00_l(ComponentPose=Local To Component(LocalPose=Use cached pose 'CRCache_Head'()), Alpha=1, GravityOverride="0,0,-980") {
    angularDampingOverride=1
    angularSpringConstant=20000
    bAngularSpring=true
    bOverrideAngularDamping=true
    bUseGravityOverride=true
    gravityScale=0.35
    simulationSpace=RootRelative
    alphaBoolBlend {blendOption=Linear}
    boundBone {boneName=dyn_pupil_00_l}
    componentLinearAccScale=<1 1 1>
    componentLinearVelScale=<1 1 1>
    gravityOverride {z=-980}
    physicsBodyDefinitions [{
      boundBone {boneName=dyn_pupil_00_l}
      boxExtents=<20 30 30>
      constraintSetup {angularLimitsMax=<15 35 35>, angularLimitsMin=<-15 -35 -35>, angularTarget=<1 0 0>}
      localJointOffset=<4 0 0>
     }]
   }
   n4 = Anim Dynamics - Bone: dyn_pupil_00_r(ComponentPose=@n3, Alpha=1, GravityOverride="0,0,-980") {
    angularDampingOverride=1
    angularSpringConstant=20000
    bAngularSpring=true
    bOverrideAngularDamping=true
    bUseGravityOverride=true
    gravityScale=0.35
    simulationSpace=RootRelative
    alphaBoolBlend {blendOption=Linear}
    boundBone {boneName=dyn_pupil_00_r}
    componentLinearAccScale=<1 1 1>
    componentLinearVelScale=<1 1 1>
    gravityOverride {z=-980}
    physicsBodyDefinitions [{
      boundBone {boneName=dyn_pupil_00_r}
      boxExtents=<20 30 30>
      constraintSetup {angularLimitsMax=<15 35 35>, angularLimitsMin=<-15 -35 -35>, angularTarget=<1 0 0>}
      localJointOffset=<4 0 0>
     }]
   }
   Output Pose(Result=Blend Poses by bool(BlendPose_0=Component To Local(ComponentPose=@n2), BlendPose_1=Component To Local(ComponentPose=@n4), BlendTime_0=0.2, BlendTime_1=0.2) {blendType=Linear})
   n5 = Control Rig!(Source=Input Pose(), Alpha, HangGliding:double=Get HangGlidingAlpha() -> (HangGlidingAlpha:double)) -> (Pose) {
    alphaBoolBlend {blendOption=Linear}
    controlRigAssetReference {blueprintRigClass=&/F_MED_UpbeatIguanaAsh_Head_CtrlRig.F_MED_UpbeatIguanaAsh_Head_CtrlRig_C}
   }
   n6 = Anim Dynamics - Bone: dyn_ears_upper02_r!(ComponentPose=Local To Component(LocalPose=@n5), Alpha=1, GravityOverride:Vector="0,0,-980") -> (Pose) {
    angularDampingOverride=1
    angularSpringConstant=25000
    bAngularSpring=true
    bOverrideAngularDamping=true
    bUseGravityOverride=true
    gravityScale=0.85
    simulationSpace=BoneRelative
    alphaBoolBlend {blendOption=Linear}
    boundBone {boneName=dyn_ears_upper02_r}
    componentAppliedLinearAccClamp=<10000 10000 10000>
    componentLinearAccScale=<1 1 1>
    componentLinearVelScale=<1 1 1>
    gravityOverride {z=-980}
    physicsBodyDefinitions [{
      boundBone {boneName=dyn_ears_upper02_r}
      boxExtents=<15.5 25.5 25.5>
      constraintSetup {angularLimitsMax=<2 3 5>, angularLimitsMin=<-2 -3 -5>, angularTarget=<1 0 0>}
      localJointOffset=<0.8 0 0>
     }]
    relativeSpaceBone {boneName=spine_03}
   }
   n7 = Anim Dynamics - Bone: dyn_ears_upper03_r!(ComponentPose=@n6, Alpha=1, GravityOverride:Vector="0,0,-980") -> (Pose) {
    angularDampingOverride=1
    angularSpringConstant=25000
    bAngularSpring=true
    bOverrideAngularDamping=true
    bUseGravityOverride=true
    gravityScale=0.85
    simulationSpace=BoneRelative
    alphaBoolBlend {blendOption=Linear}
    boundBone {boneName=dyn_ears_upper03_r}
    componentAppliedLinearAccClamp=<10000 10000 10000>
    componentLinearAccScale=<1 1 1>
    componentLinearVelScale=<1 1 1>
    gravityOverride {z=-980}
    physicsBodyDefinitions [{
      boundBone {boneName=dyn_ears_upper03_r}
      boxExtents=<15.5 25.5 25.5>
      constraintSetup {angularLimitsMax=<2 3 5>, angularLimitsMin=<-2 -3 -5>, angularTarget=<1 0 0>}
      localJointOffset=<0.8 0 0>
     }]
    relativeSpaceBone {boneName=spine_03}
   }
   n8 = Anim Dynamics - Bone: dyn_ears_mid02_r!(ComponentPose=@n7, Alpha=1, GravityOverride:Vector="0,0,-980") -> (Pose) {
    angularDampingOverride=1
    angularSpringConstant=25000
    bAngularSpring=true
    bOverrideAngularDamping=true
    bUseGravityOverride=true
    gravityScale=0.85
    simulationSpace=BoneRelative
    alphaBoolBlend {blendOption=Linear}
    boundBone {boneName=dyn_ears_mid02_r}
    componentAppliedLinearAccClamp=<10000 10000 10000>
    componentLinearAccScale=<1 1 1>
    componentLinearVelScale=<1 1 1>
    gravityOverride {z=-980}
    physicsBodyDefinitions [{
      boundBone {boneName=dyn_ears_mid02_r}
      boxExtents=<15 25 25>
      constraintSetup {angularLimitsMax=<2 3 5>, angularLimitsMin=<-2 -3 -5>, angularTarget=<1 0 0>}
      localJointOffset=<0.8 0 0>
     }]
    relativeSpaceBone {boneName=spine_03}
   }
   n9 = Anim Dynamics - Bone: dyn_ears_mid03_r!(ComponentPose=@n8, Alpha=1, GravityOverride:Vector="0,0,-980") -> (Pose) {
    angularDampingOverride=1
    angularSpringConstant=25000
    bAngularSpring=true
    bOverrideAngularDamping=true
    bUseGravityOverride=true
    gravityScale=0.85
    simulationSpace=BoneRelative
    alphaBoolBlend {blendOption=Linear}
    boundBone {boneName=dyn_ears_mid03_r}
    componentAppliedLinearAccClamp=<10000 10000 10000>
    componentLinearAccScale=<1 1 1>
    componentLinearVelScale=<1 1 1>
    gravityOverride {z=-980}
    physicsBodyDefinitions [{
      boundBone {boneName=dyn_ears_mid03_r}
      boxExtents=<15 25 25>
      constraintSetup {angularLimitsMax=<2 3 5>, angularLimitsMin=<-2 -3 -5>, angularTarget=<1 0 0>}
      localJointOffset=<0.8 0 0>
     }]
    relativeSpaceBone {boneName=spine_03}
   }
   n10 = Anim Dynamics - Bone: dyn_ears_lower02_r!(ComponentPose=@n9, Alpha=1, GravityOverride:Vector="0,0,-980") -> (Pose) {
    angularDampingOverride=1
    angularSpringConstant=25000
    bAngularSpring=true
    bOverrideAngularDamping=true
    bUseGravityOverride=true
    gravityScale=0.85
    simulationSpace=BoneRelative
    alphaBoolBlend {blendOption=Linear}
    boundBone {boneName=dyn_ears_lower02_r}
    componentAppliedLinearAccClamp=<10000 10000 10000>
    componentLinearAccScale=<1 1 1>
    componentLinearVelScale=<1 1 1>
    gravityOverride {z=-980}
    physicsBodyDefinitions [{
      boundBone {boneName=dyn_ears_lower02_r}
      boxExtents=<14.5 24.5 24.5>
      constraintSetup {angularLimitsMax=<2 3 5>, angularLimitsMin=<-2 -3 -5>, angularTarget=<1 0 0>}
      localJointOffset=<0.8 0 0>
     }]
    relativeSpaceBone {boneName=spine_03}
   }
   n11 = Anim Dynamics - Bone: dyn_ears_lower03_r!(ComponentPose=@n10, Alpha=1, GravityOverride:Vector="0,0,-980") -> (Pose) {
    angularDampingOverride=1
    angularSpringConstant=25000
    bAngularSpring=true
    bOverrideAngularDamping=true
    bUseGravityOverride=true
    gravityScale=0.85
    simulationSpace=BoneRelative
    alphaBoolBlend {blendOption=Linear}
    boundBone {boneName=dyn_ears_lower03_r}
    componentAppliedLinearAccClamp=<10000 10000 10000>
    componentLinearAccScale=<1 1 1>
    componentLinearVelScale=<1 1 1>
    gravityOverride {z=-980}
    physicsBodyDefinitions [{
      boundBone {boneName=dyn_ears_lower03_r}
      boxExtents=<14.5 24.5 24.5>
      constraintSetup {angularLimitsMax=<2 3 5>, angularLimitsMin=<-2 -3 -5>, angularTarget=<1 0 0>}
      localJointOffset=<0.8 0 0>
     }]
    relativeSpaceBone {boneName=spine_03}
   }
   n12 = Anim Dynamics - Bone: dyn_ears_lower02_l!(ComponentPose=@n11, Alpha=1, GravityOverride:Vector="0,0,-980") -> (Pose) {
    angularDampingOverride=1
    angularSpringConstant=25000
    bAngularSpring=true
    bOverrideAngularDamping=true
    bUseGravityOverride=true
    gravityScale=0.85
    simulationSpace=BoneRelative
    alphaBoolBlend {blendOption=Linear}
    boundBone {boneName=dyn_ears_lower02_l}
    componentAppliedLinearAccClamp=<10000 10000 10000>
    componentLinearAccScale=<1 1 1>
    componentLinearVelScale=<1 1 1>
    gravityOverride {z=-980}
    physicsBodyDefinitions [{
      boundBone {boneName=dyn_ears_lower02_l}
      boxExtents=<15.5 25.5 25.5>
      constraintSetup {angularLimitsMax=<2 3 5>, angularLimitsMin=<-2 -3 -5>, angularTarget=<1 0 0>}
      localJointOffset=<0.8 0 0>
     }]
    relativeSpaceBone {boneName=spine_03}
   }
   n13 = Anim Dynamics - Bone: dyn_ears_lower03_l!(ComponentPose=@n12, Alpha=1, GravityOverride:Vector="0,0,-980") -> (Pose) {
    angularDampingOverride=1
    angularSpringConstant=25000
    bAngularSpring=true
    bOverrideAngularDamping=true
    bUseGravityOverride=true
    gravityScale=0.85
    simulationSpace=BoneRelative
    alphaBoolBlend {blendOption=Linear}
    boundBone {boneName=dyn_ears_lower03_l}
    componentAppliedLinearAccClamp=<10000 10000 10000>
    componentLinearAccScale=<1 1 1>
    componentLinearVelScale=<1 1 1>
    gravityOverride {z=-980}
    physicsBodyDefinitions [{
      boundBone {boneName=dyn_ears_lower03_l}
      boxExtents=<15.5 25.5 25.5>
      constraintSetup {angularLimitsMax=<2 3 5>, angularLimitsMin=<-2 -3 -5>, angularTarget=<1 0 0>}
      localJointOffset=<0.8 0 0>
     }]
    relativeSpaceBone {boneName=spine_03}
   }
   n14 = Anim Dynamics - Bone: dyn_ears_mid02_l!(ComponentPose=@n13, Alpha=1, GravityOverride:Vector="0,0,-980") -> (Pose) {
    angularDampingOverride=1
    angularSpringConstant=25000
    bAngularSpring=true
    bOverrideAngularDamping=true
    bUseGravityOverride=true
    gravityScale=0.85
    simulationSpace=BoneRelative
    alphaBoolBlend {blendOption=Linear}
    boundBone {boneName=dyn_ears_mid02_l}
    componentAppliedLinearAccClamp=<10000 10000 10000>
    componentLinearAccScale=<1 1 1>
    componentLinearVelScale=<1 1 1>
    gravityOverride {z=-980}
    physicsBodyDefinitions [{
      boundBone {boneName=dyn_ears_mid02_l}
      boxExtents=<15 25 25>
      constraintSetup {angularLimitsMax=<2 3 5>, angularLimitsMin=<-2 -3 -5>, angularTarget=<1 0 0>}
      localJointOffset=<0.8 0 0>
     }]
    relativeSpaceBone {boneName=spine_03}
   }
   n15 = Anim Dynamics - Bone: dyn_ears_mid03_l!(ComponentPose=@n14, Alpha=1, GravityOverride:Vector="0,0,-980") -> (Pose) {
    angularDampingOverride=1
    angularSpringConstant=25000
    bAngularSpring=true
    bOverrideAngularDamping=true
    bUseGravityOverride=true
    gravityScale=0.85
    simulationSpace=BoneRelative
    alphaBoolBlend {blendOption=Linear}
    boundBone {boneName=dyn_ears_mid03_l}
    componentAppliedLinearAccClamp=<10000 10000 10000>
    componentLinearAccScale=<1 1 1>
    componentLinearVelScale=<1 1 1>
    gravityOverride {z=-980}
    physicsBodyDefinitions [{
      boundBone {boneName=dyn_ears_mid03_l}
      boxExtents=<15 25 25>
      constraintSetup {angularLimitsMax=<2 3 5>, angularLimitsMin=<-2 -3 -5>, angularTarget=<1 0 0>}
      localJointOffset=<0.8 0 0>
     }]
    relativeSpaceBone {boneName=spine_03}
   }
   n16 = Anim Dynamics - Bone: dyn_ears_upper02_l!(ComponentPose=@n15, Alpha=1, GravityOverride:Vector="0,0,-980") -> (Pose) {
    angularDampingOverride=1
    angularSpringConstant=25000
    bAngularSpring=true
    bOverrideAngularDamping=true
    bUseGravityOverride=true
    gravityScale=0.85
    simulationSpace=BoneRelative
    alphaBoolBlend {blendOption=Linear}
    boundBone {boneName=dyn_ears_upper02_l}
    componentAppliedLinearAccClamp=<10000 10000 10000>
    componentLinearAccScale=<1 1 1>
    componentLinearVelScale=<1 1 1>
    gravityOverride {z=-980}
    physicsBodyDefinitions [{
      boundBone {boneName=dyn_ears_upper02_l}
      boxExtents=<14.5 24.5 24.5>
      constraintSetup {angularLimitsMax=<2 3 5>, angularLimitsMin=<-2 -3 -5>, angularTarget=<1 0 0>}
      localJointOffset=<0.8 0 0>
     }]
    relativeSpaceBone {boneName=spine_03}
   }
   n17 = Anim Dynamics - Bone: dyn_ears_upper03_l!(ComponentPose=@n16, Alpha=1, GravityOverride:Vector="0,0,-980") -> (Pose) {
    angularDampingOverride=1
    angularSpringConstant=25000
    bAngularSpring=true
    bOverrideAngularDamping=true
    bUseGravityOverride=true
    gravityScale=0.85
    simulationSpace=BoneRelative
    alphaBoolBlend {blendOption=Linear}
    boundBone {boneName=dyn_ears_upper03_l}
    componentAppliedLinearAccClamp=<10000 10000 10000>
    componentLinearAccScale=<1 1 1>
    componentLinearVelScale=<1 1 1>
    gravityOverride {z=-980}
    physicsBodyDefinitions [{
      boundBone {boneName=dyn_ears_upper03_l}
      boxExtents=<14.5 24.5 24.5>
      constraintSetup {angularLimitsMax=<2 3 5>, angularLimitsMin=<-2 -3 -5>, angularTarget=<1 0 0>}
      localJointOffset=<0.8 0 0>
     }]
    relativeSpaceBone {boneName=spine_03}
   }
   n18 = Anim Dynamics - Bone: dyn_hair_f_l!(ComponentPose=@n17, Alpha=1, GravityOverride:Vector="0,0,-980") -> (Pose) {
    angularDampingOverride=1
    angularSpringConstant=22000
    bAngularSpring=true
    bOverrideAngularDamping=true
    bUseGravityOverride=true
    gravityScale=0.35
    simulationSpace=BoneRelative
    alphaBoolBlend {blendOption=Linear}
    boundBone {boneName=dyn_hair_f_l}
    componentAppliedLinearAccClamp=<8000 8000 8000>
    componentLinearAccScale=<1 1 1>
    componentLinearVelScale=<1 1 1>
    gravityOverride {z=-980}
    physicsBodyDefinitions [{
      boundBone {boneName=dyn_hair_f_l}
      boxExtents=<20 40 40>
      constraintSetup {angularLimitsMax=<8 8 8>, angularLimitsMin=<-8 -8 -8>, angularTarget=<1 0 0>}
      localJointOffset=<5 0 0>
     }]
    relativeSpaceBone {boneName=spine_03}
   }
   n19 = Anim Dynamics - Bone: dyn_hair_f_r!(ComponentPose=@n18, Alpha=1, GravityOverride:Vector="0,0,-980") -> (Pose) {
    angularDampingOverride=1
    angularSpringConstant=22000
    bAngularSpring=true
    bOverrideAngularDamping=true
    bUseGravityOverride=true
    gravityScale=0.35
    simulationSpace=BoneRelative
    alphaBoolBlend {blendOption=Linear}
    boundBone {boneName=dyn_hair_f_r}
    componentAppliedLinearAccClamp=<8000 8000 8000>
    componentLinearAccScale=<1 1 1>
    componentLinearVelScale=<1 1 1>
    gravityOverride {z=-980}
    physicsBodyDefinitions [{
      boundBone {boneName=dyn_hair_f_r}
      boxExtents=<20 40 40>
      constraintSetup {angularLimitsMax=<8 8 8>, angularLimitsMin=<-8 -8 -8>, angularTarget=<1 0 0>}
      localJointOffset=<5 0 0>
     }]
    relativeSpaceBone {boneName=spine_03}
   }
   n20 = Anim Dynamics - Bone: dyn_jaw!(ComponentPose=@n19, Alpha=0.65, GravityOverride:Vector="0,0,-980") -> (Pose) {
    angularDampingOverride=1
    angularSpringConstant=30000
    bAngularSpring=true
    bOverrideAngularDamping=true
    bUseGravityOverride=true
    gravityScale=0.35
    simulationSpace=BoneRelative
    alphaBoolBlend {blendOption=Linear}
    boundBone {boneName=dyn_jaw}
    componentAppliedLinearAccClamp=<8000 8000 8000>
    componentLinearAccScale=<1 1 1>
    componentLinearVelScale=<1 1 1>
    gravityOverride {z=-980}
    physicsBodyDefinitions [{
      boundBone {boneName=dyn_jaw}
      boxExtents=<20 30 30>
      constraintSetup {angularLimitsMax=<0 3 0>, angularLimitsMin=<0 -3 0>, angularTarget=<1 0 0>}
      localJointOffset=<1 0 0>
     }]
    relativeSpaceBone {boneName=neck_02}
   }
   Save cached pose 'CRCache_Head'(Pose=Component To Local(ComponentPose=@n20)) {cacheName=CRCache_Head}
  ]}
 EventGraph {nodes [
   Event BlueprintUpdateAnimation()
   > n1 = Gravity Override State Function DEPRECATED(CharacterPartAnimInstance=Self-Reference(), GameTimeInSeconds=GetGameTimeInSeconds(), GravityOverrideParams=Get Gravity Override Params())
   > Set GravityOverride(GravityOverride=@n1)
   n2 = Get bIsPlayingEmote()
   > Set Front End Menu(Front End Menu:bool=AND Boolean(A=Get bIsInFrontEnd(), B=NOT Boolean(A=@n2))) -> (Output_Get:bool)
   Event OnParachuteOpened()
   > Switch on EGliderType(Selection=Get PlayerGliderType() -> (PlayerGliderType=HangGlider))
   > Set HangGlidingAlpha(HangGlidingAlpha:double=1) -> (Output_Get:double)
   Event OnParachuteClosed()
   > Set HangGlidingAlpha(HangGlidingAlpha=0)
   TryGetPawnOwner()
  ]}
]
variables [
 Front End Menu {type=boolean}
 Gravity Override Params {type=GravityOverrideParamsStruct}
 GravityOverride {type=Vector}
 HangGlidingAlpha {type=number}
]
